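; Lightsaber behavior evaluator that decides whether an AI's behavior tree can start or keep using the melee weapon.
; Both starting and ongoing checks call their parent version, then return only the result of the shared MeleeConditions function.
; The ongoing check passes a larger threshold than the starting one, likely a range hysteresis that prevents rapid switching in and out of melee.
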(fn EvaluateOngoingConditions
  (params (OwnerComp :type Behavior%20Tree%20Component%20Object%20Reference))
  (|CallParentFunction OwnerComp)
  (bind _canusemeleeweapon (CallFunction|MeleeConditions OwnerComp 3e+03))
  (return _canusemeleeweapon))

(fn EvaluateStartingConditions
  (params (OwnerComp :type Behavior%20Tree%20Component%20Object%20Reference))
  (|CallParentFunction OwnerComp)
  (bind _canusemeleeweapon (CallFunction|MeleeConditions OwnerComp 2e+03))
  (return _canusemeleeweapon))
